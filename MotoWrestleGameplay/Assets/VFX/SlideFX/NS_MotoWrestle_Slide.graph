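# Niagara system for the MotoWrestle slide effect, leaving an icy trail behind a moving skeletal mesh.
# Ice chunks, smoke, spikes and oriented frost/ice decals spawn by distance travelled from mesh bones or sockets, gated variously by motion and ground-collision checks.
# Paired Blast, EnergyFlow and GlowFlick sprites flare on either side while User.Moving passes a threshold; User.ColorVariant tints Blast and EnergyFlow.
# Decals and spikes ramp in with emitter age.
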NiagaraSystem /MotoWrestleGameplay/Assets/VFX/SlideFX/NS_MotoWrestle_Slide
dependencies {
 usedDataInterfaces [
  &NiagaraDataInterfaceColorCurve
  &NiagaraDataInterfaceCollisionQuery
  &NiagaraDataInterfaceCurve
  &NiagaraDataInterfaceArrayColor
  &NiagaraDataInterfaceMeshRendererInfo
  &NiagaraDataInterfaceVector2DCurve
  &NiagaraDataInterfaceArrayFloat
  &NiagaraDataInterfaceSkeletalMesh
 ]
 usedDynamicInputs [&InvertBool]
 usedModules [
  &self:SystemSpawnScript.SystemScriptSource.SystemScriptGraph.NiagaraNodeAssignment_2.SetVariables_#1
  &SystemState
  &self:SystemSpawnScript.SystemScriptSource.SystemScriptGraph.NiagaraNodeAssignment_0.SetVariables_#2
  &self:CheckIfLeaveGround
  &self:SystemSpawnScript.SystemScriptSource.SystemScriptGraph.NiagaraNodeAssignment_1.SetVariables_#3
  &Timeline
  &self:SystemSpawnScript.SystemScriptSource.SystemScriptGraph.NiagaraNodeAssignment_3.SetVariables_#4
 ]
}
emitters {
 BehindIcePiece {
  emitterUpdateScript [
   &EmitterState(Life Cycle Mode=::NewEnumerator1, Loop Behavior=::NewEnumerator1, UseLoopDelay=false)
   &SpawnPerUnit(
    Module.Use Max Movement Threshold=true
    Module.Use Movement Tolerance=false
    Spawn Spacing:NiagaraFloat=&MakeCustomFloatFromBool(
     Bool=&Boolean_AndOperation(Bool1=$System.CollisionValid, Bool2=&SetBoolByFloatComparison(A=$User.Moving, B=0.001))
     True Float=50
     False Float=0
    )
    Use Spawn Probability=false
   )
  ]
  particleSpawnScript [
   &InitializeParticle(
    Color=<rgba 0.4375 0.5949 0.7965 0.1>
    Lifetime=&RandomRangeFloat(Minimum=0.4, Maximum=0.6)
    Mesh Scale=&Multiply_VectorByFloat(Vector=<1.5 0.5 0.5>, Float=&RandomRangeFloat(Minimum=0.05, Maximum=0.07))
    Mesh Scale Mode=::NewEnumerator4
    Sprite Rotation Mode=::NewEnumerator2
    Sprite Size Mode=::NewEnumerator1
    Uniform Sprite Size Max=80
    Uniform Sprite Size Min=50
    UsePositionOffset=false
    Write Material Random=true
   )
   /SetVariables(Particles.LoR:NiagaraBool=&RandomBool()) {isSetParameters=true}
   &SkeletalMeshLocation(
    Bone / Socket Index=&MakeCustomIntFromBool(Bool=$Particles.LoR, True INT=2, False INT=1)
    Bone Sampling Mode=::NewEnumerator4
    MaskSpawnGroup=false
    Orientation Sampling=::NewEnumerator1
    Skeletal Mesh=$System.TargetSKMesh
   )
   &ShapeLocation(MaskSpawnGroup=false, Override Seed=false, Sphere Radius=20, UseOwnerScale=true)
   &AddVelocity(
    Cone Angle=30
    Cone Axis=<0 0 1>
    Module.UseConeFalloff=false
    Module.UseVelDistribution=false
    Override Seed=false
    Rotation Coordinate Space=::Simulation
    Velocity Mode=::NewEnumerator2
    Velocity Speed=&RandomRangeFloat(Minimum=336.4038, Maximum=495.8982)
   ) {module=AddVelocity001}
   &AddVelocity(
    Cone Angle=15
    Cone Axis=$Engine.Owner.Velocity
    Module.UseConeFalloff=false
    Module.UseVelDistribution=false
    Override Seed=false
    Rotation Coordinate Space=::World
    Velocity Mode=::NewEnumerator2
    Velocity Speed=&RandomRangeFloat(Minimum=-100, Maximum=-200)
   ) {module=AddVelocity002}
   &AddVelocity(
    Cone Angle=15
    Cone Axis=&MakeCustomVectorFromBool(Bool=$Particles.LoR, True Vector=<1 0 0>, False Vector=<-1 0 0>)
    Module.UseConeFalloff=false
    Module.UseVelDistribution=false
    Override Seed=false
    Velocity Mode=::NewEnumerator2
    Velocity Speed=&RandomRangeFloat(Minimum=205.2992, Maximum=358.8706)
   ) {module=AddVelocity003}
   &InitialMeshOrientation(
    Mesh Orientation Mode=::NewEnumerator2
    Orientation Coordinate Space=::World
    Rotation=&RandomRangeVector(Minimum=<0 0 0>, Maximum=<1 1 1>)
    Use Rotation Vector=true
   )
  ]
  particleUpdateScript [
   &ParticleState()
   &GravityForce(Gravity=&Multiply_VectorByFloat(
     Vector=&RandomRangeVector(Minimum=<0 0 -726.6278>, Maximum=<0 0 -2545.5586>)
     Float=&FloatFromCurve(FloatCurve=DataInterface() {propertyValues {Curve {keys [<at 0.2 0>,<at 0.21 1>]}}})
    ))
   &MeshRotationForce(Rotation=&RandomRangeVector(Minimum=<0 32.952 0>, Maximum=<1 177.2451 1>))
   &ApplyOwnerScaleToAttributes(Scale Initial Mesh Scale=true)
   &SolveForcesAndVelocity(Clamp Velocity=false, Limit Acceleration=false)
   &FN_Sprite_Attributes(
    Color=&MultiplyLinearColor(Linear Color=$Particles.Color, Scale Factor=<1 1 1 0.03>)
    Create Color Attribute=true
    Create SpriteSize Attribute=true
    SpriteSize=&Multiply_Vector2D(A=$Particles.SpriteSize, B=<5 1>)
    Write to Module Namespace=true
   )
   &ScaleMeshSize(Scale Factor=&Multiply_VectorByFloat(
     Vector=<1 1 1>
     Float=&FloatFromCurve(FloatCurve=DataInterface() {propertyValues {Curve {keys [
          {arriveTangent=19.7642, interpMode=RCIM_Cubic, leaveTangent=19.7642, tangentMode=RCTM_User, time=0, value=0}
          {interpMode=RCIM_Cubic, tangentMode=RCTM_Break, time=0.15, value=0.95}
          {arriveTangent=-3.5348, interpMode=RCIM_Cubic, leaveTangent=-3.5348, tangentMode=RCTM_User, time=1, value=0}
         ]}}})
    ))
  ]
  rendererClasses [&NiagaraMeshRendererProperties]
  renderers [{rendererClass=&NiagaraMeshRendererProperties, rendererIndex=0}]
 }
 BehindSmoke {
  emitterUpdateScript [
   &EmitterState(Life Cycle Mode=::NewEnumerator1, Loop Behavior=::NewEnumerator1, UseLoopDelay=false)
   &SpawnPerUnit(
    Module.Use Max Movement Threshold=true
    Module.Use Movement Tolerance=false
    Spawn Spacing:NiagaraFloat=&MakeCustomFloatFromBool(
     Bool=&Boolean_AndOperation(Bool1=$System.CollisionValid, Bool2=&SetBoolByFloatComparison(A=$User.Moving, B=0.001))
     True Float=40
     False Float=0
    )
    Use Spawn Probability=false
   )
  ]
  particleSpawnScript [
   &InitializeParticle(
    Color=<rgba 0.455 0.5624 0.7 0.15>
    Lifetime=&RandomRangeFloat(Minimum=0.4, Maximum=0.6)
    Sprite Rotation Mode=::NewEnumerator2
    Sprite Size Mode=::NewEnumerator1
    Uniform Sprite Size Max=300
    Uniform Sprite Size Min=200
    UsePositionOffset=false
    Write Material Random=true
   )
   &SkeletalMeshLocation(
    Bone / Socket Index=&RandomRangeInt(Minimum=1, Maximum=2)
    Bone Sampling Mode=::NewEnumerator4
    MaskSpawnGroup=false
    Orientation Sampling=::NewEnumerator1
    Sampled Position Offset=<0 30 0>
    Skeletal Mesh=$System.TargetSKMesh
   )
   &ShapeLocation(MaskSpawnGroup=false, Override Seed=false, Sphere Radius=10, UseOwnerScale=true)
   &AddVelocity(
    Cone Axis=<0 0 1>
    Module.UseConeFalloff=false
    Module.UseVelDistribution=false
    Override Seed=false
    Rotation Coordinate Space=::Simulation
    Velocity Mode=::NewEnumerator2
    Velocity Speed=&RandomRangeFloat(Minimum=100, Maximum=200)
   )
   &CameraOffset(Camera Offset Amount=&RandomRangeFloat(Minimum=-30, Maximum=30))
  ]
  particleUpdateScript [
   &Drag(Drag=&RandomRangeFloat(Minimum=1, Maximum=2), Use Linear Drag=true, Use Rotational Drag=true)
   &ParticleState()
   &GravityForce(Gravity=<0 0 -300>)
   &SolveForcesAndVelocity(Clamp Velocity=false, Limit Acceleration=false)
   &ScaleSpriteSize(Uniform Curve Sprite Scale=DataInterface() {propertyValues {Curve {keys [
        {arriveTangent=20.3307, interpMode=RCIM_Cubic, leaveTangent=20.3307, tangentMode=RCTM_User, time=0, value=0}
        {arriveTangent=2.9865, interpMode=RCIM_Cubic, leaveTangent=2.9865, tangentMode=RCTM_User, time=0.05, value=0.7}
        {arriveTangent=-0.0405, interpMode=RCIM_Cubic, leaveTangent=-0.0405, tangentMode=RCTM_User, time=0.3, value=1}
       ]}}})
   &ScaleColor(
    Scale Alpha=&Multiply_Float(
     A=&FloatFromCurve(
      FloatCurve=DataInterface() {propertyValues {Curve {keys [
          {interpMode=RCIM_Cubic, leaveTangent=-6.3412, tangentMode=RCTM_Break, time=0, value=1}
          {arriveTangent=-0.6426, interpMode=RCIM_Cubic, leaveTangent=-0.6426, tangentMode=RCTM_User, time=0.1, value=0.35}
          {tangentMode=RCTM_User, time=1, value=0}
         ]}}}
      Scale Curve=&RandomRangeFloat(Minimum=0.5, Maximum=1)
     )
     B=1
    )
    Scale RGB=&VectorFromFloat(Value=&Add_Float(
      A=1
      B=&FloatFromCurve(FloatCurve=DataInterface() {propertyValues {Curve {keys [
           {interpMode=RCIM_Cubic, leaveTangent=-6.3412, tangentMode=RCTM_Break, time=0, value=1}
           {arriveTangent=-1.4208, interpMode=RCIM_Cubic, leaveTangent=-1.4208, tangentMode=RCTM_User, time=0.1, value=0.35}
           {tangentMode=RCTM_User, time=1, value=0}
          ]}}})
     ))
    ScaleA=true
    ScaleRGB=true
   )
   &SubUVAnimation()
  ]
  rendererClasses [&NiagaraSpriteRendererProperties]
  renderers [{rendererClass=&NiagaraSpriteRendererProperties}]
 }
 Blast {
  emitterUpdateScript [
   &EmitterState()
   &SpawnPerUnit(
    Module.Use Max Movement Threshold=true
    Module.Use Movement Tolerance=false
    Spawn Spacing:NiagaraFloat=&MakeCustomFloatFromBool(Bool=&SetBoolByFloatComparison(A=$User.Moving, B=0.9), True Float=300, False Float=0)
    Use Spawn Probability=false
   )
   &SpawnPerUnit(
    Module.Use Max Movement Threshold=true
    Module.Use Movement Tolerance=false
    Spawn Group=1
    Spawn Spacing:NiagaraFloat=&MakeCustomFloatFromBool(Bool=&SetBoolByFloatComparison(A=$User.Moving, B=0.9), True Float=300, False Float=0)
    Use Spawn Probability=false
   ) {module=SpawnPerUnit001}
   &SpawnPerUnit(
    Module.Use Max Movement Threshold=true
    Module.Use Movement Tolerance=false
    Spawn Spacing:NiagaraFloat=&MakeCustomFloatFromBool(Bool=&SetBoolByFloatComparison(A=$User.Moving, B=0.9), True Float=300, False Float=0)
    Use Spawn Probability=false
   ) {module=SpawnPerUnit002}
   &SpawnPerUnit(
    Module.Use Max Movement Threshold=true
    Module.Use Movement Tolerance=false
    Spawn Group=1
    Spawn Spacing:NiagaraFloat=&MakeCustomFloatFromBool(Bool=&SetBoolByFloatComparison(A=$User.Moving, B=0.9), True Float=300, False Float=0)
    Use Spawn Probability=false
   ) {module=SpawnPerUnit003}
   &SpawnPerUnit(
    Module.Use Max Movement Threshold=true
    Module.Use Movement Tolerance=false
    Spawn Spacing:NiagaraFloat=&MakeCustomFloatFromBool(Bool=&SetBoolByFloatComparison(A=$User.Moving, B=0.9), True Float=300, False Float=0)
    Use Spawn Probability=false
   ) {module=SpawnPerUnit004}
   &SpawnPerUnit(
    Module.Use Max Movement Threshold=true
    Module.Use Movement Tolerance=false
    Spawn Group=1
    Spawn Spacing:NiagaraFloat=&MakeCustomFloatFromBool(Bool=&SetBoolByFloatComparison(A=$User.Moving, B=0.9), True Float=300, False Float=0)
    Use Spawn Probability=false
   ) {module=SpawnPerUnit005}
   &SpawnBurst_Instantaneous(Use Spawn Probability=false, UseLoopCountLimit=false)
   &SpawnBurst_Instantaneous(Spawn Group=1, Use Spawn Probability=false, UseLoopCountLimit=false) {module=SpawnBurst_Instantaneous001}
   /SetVariables(Emitter.DepthFade=115.7153) {isSetParameters=true, module=SetVariables_#5}
  ]
  particleSpawnScript [
   /SetVariables(Particles.ColorShift:LinearColor=&SelectLinearColorFromArray(
     Array Sampling Mode=::NewEnumerator1
     Direct Array Index=$User.ColorVariant
     Color Selection Array=DataInterface() {propertyValues {ColorData [<rgba 0.8235 1.1273 1.5 1>, <rgba 0.625 0 0.04 1>, <rgba 1 0.3072 0 1>, <rgba 0.3522 0 1 1>]}}
     Random Seed=?
    )) {isSetParameters=true, module=SetVariables_#6}
   &InitializeParticle(
    Color=&HueShiftLinearColor(
     Hue=&RandomRangeFloat(Minimum=0.98, Maximum=1.03)
     Saturation=&RandomRangeFloat(Minimum=0.9, Maximum=2)
     Value=&RandomRangeFloat(Minimum=0.5, Maximum=5)
     Linear Color=$Particles.ColorShift
    )
    Lifetime Max=0.15
    Lifetime Min=0.15
    Lifetime Mode=::NewEnumerator1
    Position Offset=&TransformVector(
     Vector=&MaskVectorBySpawnGroup(Masked Vector=<45 120 0>, Passthrough Vector=<-45 120 0>)
     Destination Space=::Simulation
    )
    Sprite Rotation Mode=::NewEnumerator2
    Sprite Size Max=<10.1772 31.5258>
    Sprite Size Min=<7.5872 14.7881>
    Sprite Size Mode=::NewEnumerator2
    UsePositionOffset=true
    Write Material Random=true
   )
   &SubUVAnimation(End Frame=3, SubUV Animation Mode:ENiagara_SubUVLookupMode=::NewEnumerator2)
   &AddVelocity(
    Cone Angle=120
    Cone Axis=<0 0 1>
    Module.UseConeFalloff=false
    Module.UseVelDistribution=false
    Override Seed=false
    Velocity Mode=::NewEnumerator2
    Velocity Speed=&RandomRangeFloat(Minimum=50, Maximum=200)
   ) {enabled=false, module=AddVelocity002}
   &AddVelocity(
    Cone Angle=40
    Cone Axis=<0 -1 1>
    Inner Cone Angle=20
    Module.UseConeFalloff=false
    Module.UseVelDistribution=false
    Override Seed=false
    Velocity Mode=::NewEnumerator2
    Velocity Speed=100
   ) {enabled=false, module=AddVelocity001}
   &AddVelocity(
    Cone Angle=40
    Cone Axis=&MaskVectorBySpawnGroup(Masked Vector=<1.3 -1 1>, Passthrough Vector=<-1.3 -1 1>)
    Inner Cone Angle=15
    Module.UseConeFalloff=false
    Module.UseVelDistribution=false
    Override Seed=false
    Velocity Mode=::NewEnumerator2
    Velocity Speed=&RandomRangeFloat(Minimum=1, Maximum=5)
   )
   &CameraOffset()
  ]
  particleUpdateScript [
   &ParticleState()
   &SolveForcesAndVelocity(Clamp Velocity=false, Limit Acceleration=false)
   &ScaleSpriteSize(Scale Factor=&Vector2DFromCurve(
     Vector2Curve=DataInterface() {propertyValues {XCurve {keys [<at 0.0036 1.6897>,<at 0.9967 0.0582>]}, YCurve {keys [<at 0 1>,<at 1.002 3.7227>]}}}
     ScaleCurve=<5 5>
    ))
   &ScaleColor(
    Scale Alpha=&FloatFromCurve(
     FloatCurve=DataInterface() {propertyValues {Curve {keys [
         {interpMode=RCIM_Cubic, leaveTangentWeight=0.0962, tangentMode=RCTM_Break, time=0, value=0}
         {interpMode=RCIM_Cubic, tangentMode=RCTM_Break, time=0.2, value=1}
         {interpMode=RCIM_Cubic, tangentMode=RCTM_Break, time=1, value=0}
        ]}}}
     Scale Curve=&RandomRangeFloat(Minimum=0.35, Maximum=0.8)
    )
    ScaleA=true
    ScaleRGB=false
   )
   &DynamicMaterialParameters(
    Index 0 Param 2=1
    Index 0 Param 3=1
    Index 0 Param 4=&FloatFromCurve(FloatCurve=DataInterface() {propertyValues {Curve {keys [{interpMode=RCIM_Cubic, tangentMode=RCTM_Break, time=0.3, value=1}, {interpMode=RCIM_Cubic, tangentMode=RCTM_Break, time=1, value=0.25}]}}})
    Module.Param0WriteEnabled=true
    Module.Param1WriteEnabled=true
    Module.Param2WriteEnabled=true
    Module.Param3WriteEnabled=true
   )
   /SetVariables() {enabled=false, isSetParameters=true, module=SetVariables_#7}
  ]
  rendererClasses [&NiagaraSpriteRendererProperties]
  renderers [{rendererClass=&NiagaraSpriteRendererProperties}]
 }
 EnergyFlow {
  emitterUpdateScript [
   &EmitterState()
   &SpawnPerUnit(
    Module.Use Max Movement Threshold=true
    Module.Use Movement Tolerance=false
    Spawn Spacing:NiagaraFloat=&MakeCustomFloatFromBool(Bool=&SetBoolByFloatComparison(A=$User.Moving, B=0.9), True Float=300, False Float=0)
    Use Spawn Probability=false
   )
   &SpawnPerUnit(
    Module.Use Max Movement Threshold=true
    Module.Use Movement Tolerance=false
    Spawn Group=1
    Spawn Spacing:NiagaraFloat=&MakeCustomFloatFromBool(Bool=&SetBoolByFloatComparison(A=$User.Moving, B=0.9), True Float=300, False Float=0)
    Use Spawn Probability=false
   ) {module=SpawnPerUnit001}
   &SpawnPerUnit(
    Module.Use Max Movement Threshold=true
    Module.Use Movement Tolerance=false
    Spawn Spacing:NiagaraFloat=&MakeCustomFloatFromBool(Bool=&SetBoolByFloatComparison(A=$User.Moving, B=0.9), True Float=300, False Float=0)
    Use Spawn Probability=false
   ) {module=SpawnPerUnit002}
   &SpawnPerUnit(
    Module.Use Max Movement Threshold=true
    Module.Use Movement Tolerance=false
    Spawn Group=1
    Spawn Spacing:NiagaraFloat=&MakeCustomFloatFromBool(Bool=&SetBoolByFloatComparison(A=$User.Moving, B=0.9), True Float=300, False Float=0)
    Use Spawn Probability=false
   ) {module=SpawnPerUnit003}
   &SpawnPerUnit(
    Module.Use Max Movement Threshold=true
    Module.Use Movement Tolerance=false
    Spawn Spacing:NiagaraFloat=&MakeCustomFloatFromBool(Bool=&SetBoolByFloatComparison(A=$User.Moving, B=0.9), True Float=300, False Float=0)
    Use Spawn Probability=false
   ) {module=SpawnPerUnit004}
   &SpawnPerUnit(
    Module.Use Max Movement Threshold=true
    Module.Use Movement Tolerance=false
    Spawn Group=1
    Spawn Spacing:NiagaraFloat=&MakeCustomFloatFromBool(Bool=&SetBoolByFloatComparison(A=$User.Moving, B=0.9), True Float=300, False Float=0)
    Use Spawn Probability=false
   ) {module=SpawnPerUnit005}
   &SpawnBurst_Instantaneous(Use Spawn Probability=false, UseLoopCountLimit=false)
   &SpawnBurst_Instantaneous(Spawn Group=1, Use Spawn Probability=false, UseLoopCountLimit=false) {module=SpawnBurst_Instantaneous001}
   /SetVariables(Emitter.DepthFade=115.7153) {isSetParameters=true, module=SetVariables_#5}
  ]
  particleSpawnScript [
   /SetVariables(Particles.ColorShift:LinearColor=&SelectLinearColorFromArray(
     Array Sampling Mode=::NewEnumerator1
     Direct Array Index=$User.ColorVariant
     Color Selection Array=DataInterface() {propertyValues {ColorData [<rgba 0.8235 1.1273 1.5 1>, <rgba 0.625 0 0.04 1>, <rgba 1 0.3072 0 1>, <rgba 0.3522 0 1 1>]}}
     Random Seed=?
    )) {isSetParameters=true, module=SetVariables_#6}
   &InitializeParticle(
    Color=&HueShiftLinearColor(
     Hue=&RandomRangeFloat(Minimum=0.98, Maximum=1.03)
     Saturation=&RandomRangeFloat(Minimum=0.9, Maximum=2)
     Value=&RandomRangeFloat(Minimum=0.5, Maximum=5)
     Linear Color=$Particles.ColorShift
    )
    Lifetime Max=0.2
    Lifetime Min=0.15
    Lifetime Mode=::NewEnumerator1
    Position Offset=&TransformVector(
     Vector=&MaskVectorBySpawnGroup(Masked Vector=<45 120 0>, Passthrough Vector=<-45 120 0>)
     Destination Space=::Simulation
    )
    Sprite Rotation Mode=::NewEnumerator2
    Sprite Size Max=<35 45>
    Sprite Size Min=<25 40>
    Sprite Size Mode=::NewEnumerator2
    UsePositionOffset=true
    Write Material Random=true
   )
   &SubUVAnimation(End Frame=0, SubUV Animation Mode:ENiagara_SubUVLookupMode=::NewEnumerator2)
   &AddVelocity(
    Cone Angle=120
    Cone Axis=<0 0 1>
    Module.UseConeFalloff=false
    Module.UseVelDistribution=false
    Override Seed=false
    Velocity Mode=::NewEnumerator2
    Velocity Speed=&RandomRangeFloat(Minimum=50, Maximum=200)
   ) {enabled=false, module=AddVelocity002}
   &AddVelocity(
    Cone Angle=40
    Cone Axis=<0 -1 1>
    Inner Cone Angle=20
    Module.UseConeFalloff=false
    Module.UseVelDistribution=false
    Override Seed=false
    Velocity Mode=::NewEnumerator2
    Velocity Speed=100
   ) {enabled=false, module=AddVelocity001}
   &AddVelocity(
    Cone Angle=0
    Cone Axis=&MaskVectorBySpawnGroup(Masked Vector=<1.3 -1 1>, Passthrough Vector=<-1.3 -1 1>)
    Module.UseConeFalloff=false
    Module.UseVelDistribution=false
    Override Seed=false
    Velocity Mode=::NewEnumerator2
    Velocity Speed=&RandomRangeFloat(Minimum=800, Maximum=900)
   )
   &CameraOffset()
  ]
  particleUpdateScript [
   &ParticleState()
   &SolveForcesAndVelocity(Clamp Velocity=false, Limit Acceleration=false)
   &ScaleSpriteSize(Scale Factor=&Vector2DFromCurve(
     Vector2Curve=DataInterface() {propertyValues {XCurve {keys [<at 0.0036 0.5>,<at 0.9967 1>]}, YCurve {keys [<at 0 0.5>,<at 1.002 2>]}}}
     ScaleCurve=<5 5>
    ))
   &ScaleColor(
    Scale Alpha=&FloatFromCurve(
     FloatCurve=DataInterface() {propertyValues {Curve {keys [
         {interpMode=RCIM_Cubic, leaveTangent=0.0423, leaveTangentWeight=0.0962, tangentMode=RCTM_Break, time=0, value=0}
         {interpMode=RCIM_Cubic, leaveTangent=-0.0099, tangentMode=RCTM_Break, time=0.4, value=1}
         {interpMode=RCIM_Cubic, tangentMode=RCTM_Break, time=1, value=0}
        ]}}}
     Scale Curve=&RandomRangeFloat(Minimum=0.05, Maximum=0.07)
    )
    ScaleA=true
    ScaleRGB=false
   )
   &DynamicMaterialParameters(
    Index 0 Param 2=1
    Index 0 Param 3=1
    Index 0 Param 4=&FloatFromCurve(FloatCurve=DataInterface() {propertyValues {Curve {keys [{interpMode=RCIM_Cubic, tangentMode=RCTM_Break, time=0.3, value=1}, {interpMode=RCIM_Cubic, tangentMode=RCTM_Break, time=1, value=0.25}]}}})
    Module.Param0WriteEnabled=true
    Module.Param1WriteEnabled=true
    Module.Param2WriteEnabled=true
    Module.Param3WriteEnabled=true
   )
   /SetVariables() {enabled=false, isSetParameters=true, module=SetVariables_#7}
  ]
  rendererClasses [&NiagaraSpriteRendererProperties]
  renderers [{rendererClass=&NiagaraSpriteRendererProperties}]
 }
 FrostDecal {
  emitterUpdateScript [
   &EmitterState(Life Cycle Mode=::NewEnumerator1, Loop Behavior=::NewEnumerator1, Loop Duration=1, UseLoopDelay=false)
   &SpawnBurst_Instantaneous(Use Spawn Probability=false, UseLoopCountLimit=false) {enabled=false}
   &SpawnPerUnit(
    Module.Use Max Movement Threshold=true
    Module.Use Movement Tolerance=false
    Spawn Spacing:NiagaraFloat=&MakeCustomFloatFromBool(Bool=&SetBoolByFloatComparison(A=$System.M, B=0.01), True Float=220, False Float=0)
    Use Spawn Probability=false
   )
   /SetVariables(Emitter.GradTime=0.5, Emitter.MinWidth=-0.4) {isSetParameters=true, module=SetVariables_#8}
  ]
  particleSpawnScript [
   &InitializeParticle(
    Color=&ColorFromCurve(
     ColorCurve=DataInterface() {propertyValues {
       AlphaCurve {keys [<at 0.0686 0>,<at 0.1811 0.5>,<at 0.2298 1 RCIM_Cubic>]}
       BlueCurve {keys [<at 0 1 RCIM_Cubic>,<at 1 1 RCIM_Cubic>]}
       GreenCurve {keys [<at 0 1 RCIM_Cubic>,<at 1 1 RCIM_Cubic>]}
       RedCurve {keys [<at 0 1 RCIM_Cubic>,<at 1 1 RCIM_Cubic>]}
      }}
     CurveIndex=$Emitter.Age
    )
    Lifetime=3
    Mesh Scale=<4 4 1>
    Mesh Scale Mode=::NewEnumerator4
    Ribbon Width=300
    Ribbon Width Mode=::NewEnumerator1
    UsePositionOffset=false
    Write Material Random=true
   )
   &SkeletalMeshLocation(
    Bone Sampling Mode=::NewEnumerator4
    MaskSpawnGroup=false
    Orientation Sampling=::NewEnumerator1
    Sampled Position Offset=<0 60 0>
    Skeletal Mesh=$System.TargetSKMesh
   )
   &InitialMeshOrientation(Mesh Orientation Mode=::NewEnumerator2, Orientation Vector=$Engine.Owner.Velocity, Use Rotation Vector=false) {enabled=false}
   &self:ScratchModule() {enabled=false, module=ScratchModule}
   &self:CalcTheDecalRotation(Collider:Vector3f=$Particles.CollisionNormal) {module=CalcTheDecalRotation}
   &KillParticles(Kill Particles=$Particles.VerticleOne)
   &self:LerpTip(D=2) {module=LerpTip}
   &DynamicMaterialParameters(
    Index 2 Param 2=$Particles.F
    Module.Index2Param0Write=true
    Module.Index2Param1Write=true
    Module.Index2Param2Write=true
    Module.Index2Param3Write=true
    Write Parameter Index 0=false
    Write Parameter Index 2=true
   ) {module=DynamicMaterialParameters001}
  ]
  particleUpdateScript [
   &ParticleState()
   &self:LerpTip(D=2) {module=LerpTip001}
   &self:QuatToAxisAngle(Quat:Quat4f=$Particles.MeshOrientation) {module=QuatToAxisAngle}
   &self:SetTipAndTailOpacity() {module=SetTipAndTailOpacity}
   &DynamicMaterialParameters(
    Index 0 Param 1=&MakeFloatFromVector(Vector=$Particles.Axis, Channel=::NewEnumerator0)
    Index 0 Param 2=&MakeFloatFromVector(Vector=$Particles.Axis, Channel=::NewEnumerator1)
    Index 0 Param 3=&MakeFloatFromVector(Vector=$Particles.Axis, Channel=::NewEnumerator2)
    Index 0 Param 4=$Particles.Angle
    Index 1 Param 1=&Lerp_Float(
     A=150
     B=350
     Alpha=&ClampFloat(Float=&Divide_Float(A=$Emitter.Age, B=$Emitter.GradTime))
    )
    Index 1 Param 2=&Lerp_Float(
     A=150
     B=350
     Alpha=&ClampFloat(Float=&Divide_Float(A=$Emitter.Age, B=$Emitter.GradTime))
    )
    Index 1 Param 3=100
    Index 1 Param 4=&Lerp_Float(
     A=$Particles.MyStartWidth
     B=$Emitter.MinWidth
     Alpha=&FloatFromCurve(FloatCurve=DataInterface() {propertyValues {Curve {keys [<at 0.5 0>,<at 1 1>]}}})
    )
    Index 2 Param 2=$Particles.F
    Index 3 Param 1=&FloatFromCurve(FloatCurve=DataInterface() {propertyValues {Curve {keys [<at 0 1>,<at 1 -5>]}}})
    Module.Index0Param0Write=true
    Module.Index0Param1Write=true
    Module.Index0Param2Write=true
    Module.Index0Param3Write=true
    Module.Index1Param0Write=true
    Module.Index1Param1Write=true
    Module.Index1Param2Write=true
    Module.Index1Param3Write=true
    Module.Index2Param0Write=true
    Module.Index2Param1Write=true
    Module.Index2Param2Write=true
    Module.Index2Param3Write=true
    Module.Index3Param0Write=true
    Module.Index3Param1Write=true
    Module.Index3Param2Write=true
    Module.Index3Param3Write=true
    Write Parameter Index 1=true
    Write Parameter Index 2=true
    Write Parameter Index 3=true
   )
   &ScaleColor(
    Scale Alpha=&FloatFromCurve(
     FloatCurve=DataInterface() {propertyValues {Curve {keys [
         {interpMode=RCIM_Cubic, leaveTangent=1.5405, tangentMode=RCTM_Break, time=0, value=0}
         {arriveTangent=0.4353, tangentMode=RCTM_User, time=1, value=1}
        ]}}}
     CurveIndex=&Divide_Float(A=$Emitter.Age, B=$Emitter.GradTime)
     Scale Curve=1.5
    )
    ScaleA=true
    ScaleRGB=true
   )
   &ScaleColor(
    Scale Alpha=&FloatFromCurve(FloatCurve=DataInterface() {propertyValues {Curve {keys [
         {tangentMode=RCTM_User, time=0, value=1}
         {arriveTangent=-0.0405, interpMode=RCIM_Cubic, leaveTangent=-0.0405, tangentMode=RCTM_User, time=0.7009, value=1}
         {arriveTangent=-8.0089, interpMode=RCIM_Cubic, leaveTangent=-8.0089, tangentMode=RCTM_User, time=1, value=0}
        ]}}})
    ScaleA=true
    ScaleRGB=true
   ) {module=ScaleColor001}
  ]
  rendererClasses [&NiagaraRibbonRendererProperties, &NiagaraMeshRendererProperties]
  renderers [
   {rendererClass=&NiagaraRibbonRendererProperties}
   {rendererClass=&NiagaraMeshRendererProperties, rendererIndex=1}
   {rendererClass=&NiagaraMeshRendererProperties, rendererIndex=2}
  ]
 }
 GlowFlick {
  emitterSpawnScript [
   /SetVariables(Emitter.CameraOffsetSprite=200) {isSetParameters=true}
   /SetVariables(Emitter.DecalColor:LinearColor=<rgba 0.4424 1 0.0908 1>) {isSetParameters=true}
  ]
  emitterUpdateScript [
   &EmitterState()
   &SpawnPerUnit(
    Module.Use Max Movement Threshold=true
    Module.Use Movement Tolerance=false
    Spawn Spacing:NiagaraFloat=&MakeCustomFloatFromBool(Bool=&SetBoolByFloatComparison(A=$User.Moving, B=0.9), True Float=400, False Float=0)
    Use Spawn Probability=false
   )
   &SpawnPerUnit(
    Module.Use Max Movement Threshold=true
    Module.Use Movement Tolerance=false
    Spawn Group=1
    Spawn Spacing:NiagaraFloat=&MakeCustomFloatFromBool(Bool=&SetBoolByFloatComparison(A=$User.Moving, B=0.9), True Float=400, False Float=0)
    Use Spawn Probability=false
   ) {module=SpawnPerUnit001}
   &SpawnBurst_Instantaneous(Use Spawn Probability=false, UseLoopCountLimit=false)
   &SpawnBurst_Instantaneous(Spawn Group=1, Use Spawn Probability=false, UseLoopCountLimit=false) {module=SpawnBurst_Instantaneous003}
   &SpawnBurst_Instantaneous(Use Spawn Probability=false, UseLoopCountLimit=true) {enabled=false, module=SpawnBurst_Instantaneous002}
   &SpawnBurst_Instantaneous(Spawn Group=1, Use Spawn Probability=false, UseLoopCountLimit=false) {enabled=false, module=SpawnBurst_Instantaneous001}
   &SpawnPerFrame(Spawn:NiagaraBool=$System.ShouldSpawn, Use Spawn Probability=false) {enabled=false}
   &SpawnPerFrame(Spawn:NiagaraBool=$System.ShouldSpawn, Spawn Group=1, Use Spawn Probability=false) {enabled=false, module=SpawnPerFrame001}
  ]
  particleSpawnScript [
   &InitializeParticle(
    Color=<rgba 0.25 0.3333 0.5 0.05>
    Lifetime=&Multiply_Float(
     A=&SelectFloatFromArray(Direct Array Index=$Particles.VisibilityTag, Float Selection Array=DataInterface() {propertyValues {FloatData [0.2,0.2]}})
     B=1
    )
    Position Offset=&TransformVector(
     Vector=&MaskVectorBySpawnGroup(Masked Vector=<50 90 0>, Passthrough Vector=<-50 90 0>)
     Destination Space=::Simulation
    )
    Sprite Rotation Mode=::NewEnumerator2
    Sprite Size Mode=::NewEnumerator3
    Uniform Sprite Size=319.0536
    UsePositionOffset=true
    Write Material Random=true
   )
   &DynamicMaterialParameters(Module.Index0Param0Write=true, Module.Index0Param1Write=true, Module.Index0Param2Write=true, Module.Index0Param3Write=true)
  ]
  particleUpdateScript [
   &ParticleState()
   &CameraOffset(Camera Offset Amount=&FN_LOD_Distance_Remap(Start Distance=0, End Distance=0, Minimum Output=0, Maximum Output=0, Clamp Output=true))
   &ApplyOwnerScaleToAttributes(Scale Initial Sprite Size=true)
   &ScaleSpriteSize(Uniform Curve Sprite Scale=DataInterface() {propertyValues {Curve {keys [{tangentMode=RCTM_User, time=0, value=0.7}, {tangentMode=RCTM_User, time=1, value=1.5}]}}})
   &ScaleColor(
    Scale Alpha=&FloatFromCurve(FloatCurve=DataInterface() {propertyValues {Curve {keys [{interpMode=RCIM_Cubic, tangentMode=RCTM_User, time=0, value=1}, {interpMode=RCIM_Cubic, tangentMode=RCTM_User, time=1, value=0}]}}})
    Scale RGB=&VectorFromFloat(Value=&Lerp_Float(
      A=1
      B=3
      Alpha=&FloatFromCurve(FloatCurve=DataInterface() {propertyValues {Curve {keys [
           {arriveTangent=-2.5249, interpMode=RCIM_Cubic, leaveTangent=-2.5249, tangentMode=RCTM_User, time=0.05, value=20}
           {arriveTangent=-1.146, time=0.4, value=0.1}
           {arriveTangent=-0.0405, interpMode=RCIM_Cubic, leaveTangent=-0.0405, tangentMode=RCTM_User, time=1, value=0}
          ]}}})
     ))
    ScaleA=true
    ScaleRGB=true
   ) {module=ScaleColor002}
   &SolveForcesAndVelocity(Clamp Velocity=false, Limit Acceleration=false)
  ]
  rendererClasses [&NiagaraSpriteRendererProperties]
  renderers [{rendererClass=&NiagaraSpriteRendererProperties}]
 }
 IceDecal {
  emitterUpdateScript [
   &EmitterState(Life Cycle Mode=::NewEnumerator1, Loop Behavior=::NewEnumerator1, Loop Duration=1, UseLoopDelay=false)
   &SpawnBurst_Instantaneous(Use Spawn Probability=false, UseLoopCountLimit=false) {enabled=false}
   &SpawnPerUnit(
    Module.Use Max Movement Threshold=true
    Module.Use Movement Tolerance=false
    Spawn Spacing:NiagaraFloat=&MakeCustomFloatFromBool(Bool=&SetBoolByFloatComparison(A=$System.M, B=0.01), True Float=140, False Float=0)
    Use Spawn Probability=false
   )
   /SetVariables(Emitter.GradTime=0.5, Emitter.MinWidth=-0.4) {isSetParameters=true, module=SetVariables_#8}
  ]
  particleSpawnScript [
   &InitializeParticle(
    Color=&ColorFromCurve(
     ColorCurve=DataInterface() {propertyValues {
       AlphaCurve {keys [<at 0.0686 0>,<at 0.1811 0.5>,<at 0.2298 1 RCIM_Cubic>]}
       BlueCurve {keys [<at 0 1 RCIM_Cubic>,<at 1 1 RCIM_Cubic>]}
       GreenCurve {keys [<at 0 1 RCIM_Cubic>,<at 1 1 RCIM_Cubic>]}
       RedCurve {keys [<at 0 1 RCIM_Cubic>,<at 1 1 RCIM_Cubic>]}
      }}
     CurveIndex=$Emitter.Age
    )
    Lifetime=3
    Mesh Scale=<3 3 1>
    Mesh Scale Mode=::NewEnumerator4
    Ribbon Width=187.511
    Ribbon Width Mode=::NewEnumerator1
    UsePositionOffset=false
    Write Material Random=true
   )
   &SkeletalMeshLocation(
    Bone Sampling Mode=::NewEnumerator4
    MaskSpawnGroup=false
    Orientation Sampling=::NewEnumerator1
    Sampled Position Offset=<0 60 0>
    Skeletal Mesh=$System.TargetSKMesh
   )
   &InitialMeshOrientation(Mesh Orientation Mode=::NewEnumerator2, Orientation Vector=$Engine.Owner.Velocity, Use Rotation Vector=false) {enabled=false}
   &self:ScratchModule() {enabled=false, module=ScratchModule}
   &self:CalcTheDecalRotation(Collider:Vector3f=$Particles.CollisionNormal) {module=CalcTheDecalRotation}
   &KillParticles(Kill Particles=$Particles.VerticleOne)
   /SetVariables(Particles.MyStartWidth:NiagaraFloat=/Game/Effects/Niagara/DynamicInputs/Math/FN_Fit(
     Clamp:NiagaraBool=true
     Value:NiagaraFloat=&FloatFromCurve(
      FloatCurve=DataInterface() {propertyValues {Curve {keys [
          {interpMode=RCIM_Cubic, leaveTangent=3.5167, tangentMode=RCTM_Break, time=0, value=0}
          {arriveTangent=1, interpMode=RCIM_Cubic, leaveTangent=1, time=0.5, value=0.8}
          {arriveTangent=0.4353, tangentMode=RCTM_User, time=1, value=1}
         ]}}}
      CurveIndex=&Divide_Float(A=$Emitter.Age, B=$Emitter.GradTime)
     )
     SourceMin:NiagaraFloat=0
     SourceMax:NiagaraFloat=1
     DestMin:NiagaraFloat=$Emitter.MinWidth
     DestMax:NiagaraFloat=0
    )) {isSetParameters=true}
   /SetVariables(Particles.bIsDoonce=true) {isSetParameters=true}
   &self:LerpTip(D=2) {module=LerpTip}
   &DynamicMaterialParameters(
    Index 2 Param 2=$Particles.F
    Module.Index2Param0Write=true
    Module.Index2Param1Write=true
    Module.Index2Param2Write=true
    Module.Index2Param3Write=true
    Write Parameter Index 0=false
    Write Parameter Index 2=true
   ) {module=DynamicMaterialParameters001}
  ]
  particleUpdateScript [
   &ParticleState()
   &self:LerpTip(D=2) {module=LerpTip001}
   /SetVariables() {isSetParameters=true}
   &self:QuatToAxisAngle(Quat:Quat4f=$Particles.MeshOrientation) {module=QuatToAxisAngle}
   &self:SetTipAndTailOpacity() {module=SetTipAndTailOpacity}
   &DynamicMaterialParameters(
    Index 0 Param 1=&MakeFloatFromVector(Vector=$Particles.Axis, Channel=::NewEnumerator0)
    Index 0 Param 2=&MakeFloatFromVector(Vector=$Particles.Axis, Channel=::NewEnumerator1)
    Index 0 Param 3=&MakeFloatFromVector(Vector=$Particles.Axis, Channel=::NewEnumerator2)
    Index 0 Param 4=$Particles.Angle
    Index 1 Param 1=250
    Index 1 Param 2=250
    Index 1 Param 3=100
    Index 1 Param 4=&Lerp_Float(
     A=$Particles.MyStartWidth
     B=$Emitter.MinWidth
     Alpha=&FloatFromCurve(FloatCurve=DataInterface() {propertyValues {Curve {keys [<at 0.5 0>,<at 1 1>]}}})
    )
    Index 2 Param 2=$Particles.F
    Module.Index0Param0Write=true
    Module.Index0Param1Write=true
    Module.Index0Param2Write=true
    Module.Index0Param3Write=true
    Module.Index1Param0Write=true
    Module.Index1Param1Write=true
    Module.Index1Param2Write=true
    Module.Index1Param3Write=true
    Module.Index2Param0Write=true
    Module.Index2Param1Write=true
    Module.Index2Param2Write=true
    Module.Index2Param3Write=true
    Write Parameter Index 1=true
    Write Parameter Index 2=true
   )
   &DoOnce(Trigger Condition=true)
   &ScaleColor(
    Scale Alpha=&FloatFromCurve(
     FloatCurve=DataInterface() {propertyValues {Curve {keys [
         {interpMode=RCIM_Cubic, leaveTangent=1.5405, tangentMode=RCTM_Break, time=0, value=0}
         {arriveTangent=0.4353, tangentMode=RCTM_User, time=1, value=1}
        ]}}}
     CurveIndex=&Divide_Float(A=$Emitter.Age, B=$Emitter.GradTime)
    )
    ScaleA=true
    ScaleRGB=true
   )
   /SetVariables(Particles.bIsDoonce:NiagaraBool=$Particles.DoOnce.Execute) {isSetParameters=true}
  ]
  rendererClasses [&NiagaraRibbonRendererProperties, &NiagaraMeshRendererProperties]
  renderers [
   {rendererClass=&NiagaraRibbonRendererProperties}
   {rendererClass=&NiagaraMeshRendererProperties, rendererIndex=1}
   {rendererClass=&NiagaraMeshRendererProperties, rendererIndex=2}
  ]
 }
 RandomSpike {
  emitterUpdateScript [
   &EmitterState(Life Cycle Mode=::NewEnumerator1, Loop Behavior=::NewEnumerator1, UseLoopDelay=true)
   /SetVariables(Emitter.GradTime=0.5, Emitter.MinWidth=-0.4) {isSetParameters=true}
   /SetVariables(Emitter.float:NiagaraFloat=&FloatFromCurve(
     FloatCurve=DataInterface() {propertyValues {Curve {keys [
         {interpMode=RCIM_Cubic, leaveTangent=3.5167, tangentMode=RCTM_Break, time=0, value=0}
         {arriveTangent=1, interpMode=RCIM_Cubic, leaveTangent=1, time=0.5, value=0.8}
         {arriveTangent=0.4353, tangentMode=RCTM_User, time=1, value=1}
        ]}}}
     CurveIndex=&Divide_Float(A=$Emitter.Age, B=$Emitter.GradTime)
    )) {isSetParameters=true}
   &SpawnPerUnit(
    Module.Use Max Movement Threshold=true
    Module.Use Movement Tolerance=false
    Spawn Spacing:NiagaraFloat=&MakeCustomFloatFromBool(
     Bool=&Boolean_AndOperation(Bool1=&SetBoolByFloatComparison(A=$System.M, B=0.01), Bool2=$System.CollisionValid)
     True Float=&Lerp_Float(A=150, B=40, Alpha=$Emitter.float)
     False Float=0
    )
    Use Spawn Probability=false
   )
   &SpawnPerUnit(
    Module.Use Max Movement Threshold=true
    Module.Use Movement Tolerance=false
    Spawn Group=1
    Spawn Spacing:NiagaraFloat=&MakeCustomFloatFromBool(
     Bool=&Boolean_AndOperation(Bool1=&SetBoolByFloatComparison(A=$System.M, B=0.01), Bool2=$System.CollisionValid)
     True Float=&Lerp_Float(A=150, B=40, Alpha=$Emitter.float)
     False Float=0
    )
    Use Spawn Probability=false
   ) {module=SpawnPerUnit001}
  ]
  particleSpawnScript [
   &InitializeParticle(
    Color Mode=::NewEnumerator0
    Lifetime=&RandomRangeFloat(Minimum=2.25, Maximum=2.7)
    Mesh Scale=&Multiply_VectorByFloat(
     Vector=&RandomRangeVector(Minimum=<0.5 0.5 1>, Maximum=<0.8 0.8 2>)
     Float=&Multiply_Float(
      A=&RandomRangeFloat(Minimum=0.25, Maximum=0.5)
      B=&FN_Fit(
       Clamp:NiagaraBool=true
       Value:NiagaraFloat=$Emitter.float
       SourceMin:NiagaraFloat=0
       SourceMax:NiagaraFloat=1
       DestMin:NiagaraFloat=0.2
       DestMax:NiagaraFloat=1
      )
     )
    )
    Mesh Scale Mode=::NewEnumerator4
    Position Offset=<0 0 -50>
    UsePositionOffset=true
    Write Material Random=true
   )
   /SetVariables(Particles.LoR:NiagaraBool=/Niagara/DynamicInputs/Helpers/MaskBool(
     Mask Parameter:NiagaraInt32=&SpawnGroup()
     Mask:NiagaraInt32=0
     Masked Bool:NiagaraBool=false
     Passthrough Bool:NiagaraBool=true
    )) {isSetParameters=true}
   &SkeletalMeshLocation(
    Bone / Socket Index=&MakeCustomIntFromBool(Bool=$Particles.LoR, True INT=2, False INT=1)
    Bone Sampling Mode=::NewEnumerator4
    MaskSpawnGroup=false
    Orientation Sampling=::NewEnumerator1
    Skeletal Mesh=$System.TargetSKMesh
   )
   &ShapeLocation(
    MaskSpawnGroup=false
    Non Uniform Scale=<1 1 0.001>
    Offset=&Add_Vector(
     A=<0 0 -20>
     B=&Multiply_VectorByFloat(
      Vector=&NormalizeVector(Vector To Normalize=$Engine.Owner.Velocity, Safe Normalize=true, Safe Fallback Threshold=0, Safe Fallback Vector=<0 0 0.001>)
      Float=50
     )
    )
    Offset Coordinate Space=::World
    Offset Mode=::NewEnumerator0
    Override Seed=false
    Sphere Radius=&FN_Fit(
     Clamp:NiagaraBool=true
     Value:NiagaraFloat=&FloatFromCurve(
      FloatCurve=DataInterface() {propertyValues {Curve {keys [
          {interpMode=RCIM_Cubic, leaveTangent=3.5167, tangentMode=RCTM_Break, time=0, value=0}
          {arriveTangent=1, interpMode=RCIM_Cubic, leaveTangent=1, time=0.5, value=0.8}
          {arriveTangent=0.4353, tangentMode=RCTM_User, time=1, value=1}
         ]}}}
      CurveIndex=&Divide_Float(A=$Emitter.Age, B=$Emitter.GradTime)
     )
     SourceMin:NiagaraFloat=0
     SourceMax:NiagaraFloat=1
     DestMin:NiagaraFloat=5
     DestMax:NiagaraFloat=30
    )
    UseOwnerScale=true
   )
   &AddVelocity(
    Cone Angle=30
    Cone Axis=<0 0 1>
    Module.UseConeFalloff=false
    Module.UseVelDistribution=false
    Override Seed=false
    Rotation Coordinate Space=::Simulation
    Velocity Mode=::NewEnumerator2
    Velocity Speed=&RandomRangeFloat(Minimum=138.4529, Maximum=200)
   ) {module=AddVelocity001}
   &AddVelocity(
    Cone Angle=15
    Cone Axis=$Engine.Owner.Velocity
    Module.UseConeFalloff=false
    Module.UseVelDistribution=false
    Override Seed=false
    Rotation Coordinate Space=::World
    Velocity Mode=::NewEnumerator2
    Velocity Speed=&RandomRangeFloat(Minimum=-452.4704, Maximum=-578.7618)
   ) {module=AddVelocity002}
   &AddVelocity(
    Cone Angle=15
    Cone Axis=&MakeCustomVectorFromBool(Bool=$Particles.LoR, True Vector=<1 0 0>, False Vector=<-1 0 0>)
    Module.UseConeFalloff=false
    Module.UseVelDistribution=false
    Override Seed=false
    Velocity Mode=::NewEnumerator2
    Velocity Speed=&RandomRangeFloat(Minimum=151.7257, Maximum=258.2541)
   ) {module=AddVelocity003}
   /SetVariables(Particles.StartVelocity:Vector3f=$Engine.Owner.Velocity) {isSetParameters=true}
  ]
  particleUpdateScript [
   &SolveForcesAndVelocity(Clamp Velocity=true, Limit Acceleration=false, Speed Limit=0.001)
   &ParticleState()
   &ScaleMeshSize(Scale Factor=&VectorFromFloat(Value=&FloatFromCurve(FloatCurve=DataInterface() {propertyValues {Curve {keys [
          {arriveTangent=0.4667, interpMode=RCIM_Cubic, leaveTangent=0.0691, tangentMode=RCTM_Break, time=0.05, value=0.9847}
          {arriveTangent=-0.1078, interpMode=RCIM_Cubic, leaveTangent=-0.1078, tangentMode=RCTM_User, time=0.5, value=1}
          {arriveTangent=-0.0405, interpMode=RCIM_Cubic, leaveTangent=-0.0405, tangentMode=RCTM_User, time=1, value=0}
         ]}}})))
   /SetVariables(Particles.MeshOrientation=&LerpQuaternion(
     Quaternion A:Quat4f=&FindQuatBetweenTwoVectors(
      From Vector=<0 0 1>
      To Vector=&NormalizeVector(
       Vector To Normalize=&Multiply_VectorByFloat(Vector=$Particles.StartVelocity, Float=-1)
       Safe Normalize=true
       Safe Fallback Threshold=0
       Safe Fallback Vector=<0 0 0.001>
      )
     )
     Quaternion B:Quat4f=&FindQuatBetweenTwoVectors(
      From Vector=<0 0 1>
      To Vector=&NormalizeVector(Vector To Normalize=$Particles.Velocity, Safe Normalize=true, Safe Fallback Threshold=0, Safe Fallback Vector=<0 0 0.001>)
     )
     Lerp Factor:NiagaraFloat=&FloatFromCurve(FloatCurve=DataInterface() {propertyValues {Curve {keys [{tangentMode=RCTM_User, time=0.1, value=1}]}}})
     Enable Spherical Linear Interpolation:NiagaraBool=false
    )) {isSetParameters=true}
   &DynamicMaterialParameters(
    Index 0 Param 1=&RandomRangeFloat(Minimum=0, Maximum=100)
    Module.Index0Param0Write=true
    Module.Index0Param1Write=true
    Module.Index0Param2Write=true
    Module.Index0Param3Write=true
   )
  ]
  rendererClasses [&NiagaraSpriteRendererProperties, &NiagaraMeshRendererProperties]
  renderers [
   {rendererClass=&NiagaraSpriteRendererProperties}
   {rendererClass=&NiagaraMeshRendererProperties, rendererIndex=1}
  ]
 }
 RollingMesh {
  emitterUpdateScript [
   &EmitterState(Life Cycle Mode=::NewEnumerator1, Loop Behavior=::NewEnumerator1, UseLoopDelay=false)
   &SpawnBurst_Instantaneous(Spawn Count=10, Use Spawn Probability=false, UseLoopCountLimit=false)
   &SpawnBurst_Instantaneous(Spawn Count=10, Spawn Group=1, Use Spawn Probability=false, UseLoopCountLimit=false) {module=SpawnBurst_Instantaneous001}
   /SetVariables(Emitter.GradTime=0.6, Emitter.MinWidth=-0.4) {isSetParameters=true}
   /SetVariables(
    Emitter.float:NiagaraFloat=&FloatFromCurve(
     FloatCurve=DataInterface() {propertyValues {Curve {keys [
         {interpMode=RCIM_Cubic, leaveTangent=3.5167, tangentMode=RCTM_Break, time=0, value=0}
         {arriveTangent=0.0009, tangentMode=RCTM_User, time=1, value=1}
        ]}}}
     CurveIndex=&Divide_Float(A=$Emitter.Age, B=$Emitter.GradTime)
    )
    Emitter.float001:NiagaraFloat=&FloatFromCurve(
     FloatCurve=DataInterface() {propertyValues {Curve {keys [
         {interpMode=RCIM_Cubic, leaveTangent=0.3902, tangentMode=RCTM_Break, time=1.5, value=0}
         {arriveTangent=0.0009, tangentMode=RCTM_User, time=2.5, value=1}
        ]}}}
     CurveIndex=&Divide_Float(A=$Emitter.Age, B=$Emitter.GradTime)
    )
   ) {isSetParameters=true}
  ]
  particleSpawnScript [
   &InitializeParticle(
    Lifetime=0.6
    Mesh Scale Mode=::NewEnumerator3
    Mesh Uniform Scale=&RandomRangeFloat(Minimum=0.8, Maximum=1)
    UsePositionOffset=false
    Write Material Random=true
   )
   /SetVariables(Particles.Progress:NiagaraFloat=&ReturnNormalizedExecIndex()) {isSetParameters=true}
   &ShapeLocation(
    MaskSpawnGroup=false
    Non Uniform Scale=<1 1 0>
    Offset=<0 130 0>
    Offset Coordinate Space=::Local
    Offset Mode=::NewEnumerator0
    Override Seed=false
    Sphere Radius=10
    UseOwnerScale=true
   )
   /SetVariables(Particles.LSOffset:Vector3f=&TransformPosition(
     Position=Linked() {linkedVariable {name=Particles.Initial.ShapeLocation.ShapePosition, type=Vector3f}}
     Source Space=::World
     Destination Space=::Local
    )) {isSetParameters=true}
   /SetVariables() {isSetParameters=true}
  ]
  particleUpdateScript [
   &ParticleState(Kill Particles When Lifetime Has Elapsed=false)
   &ScaleMeshSize(Scale Factor=&MakeVector(
     X=0.8
     Y=0.8
     Z=&FloatFromCurve(FloatCurve=DataInterface() {propertyValues {Curve {keys [<at 0 1>,<at 0.05 1.2>,<at 0.3 1>]}}}, CurveIndex=$Particles.Progress)
    )) {module=ScaleMeshSize004}
   &ScaleMeshSize(Scale Factor=&VectorFromFloat(Value=/Game/Effects/Niagara/DynamicInputs/Math/FN_Fit01(Clamp:NiagaraBool=false, Value:NiagaraFloat=$Emitter.float, DestMin:NiagaraFloat=0, DestMax:NiagaraFloat=1))) {module=ScaleMeshSize001}
   &ScaleMeshSize(Scale Factor=&VectorFromFloat(Value=&FloatFromCurve(
      FloatCurve=DataInterface() {propertyValues {Curve {keys [
          {interpMode=RCIM_Cubic, leaveTangent=8.0598, tangentMode=RCTM_Break, time=0, value=0}
          {arriveTangent=-0.9971, interpMode=RCIM_Cubic, leaveTangent=-2.726, tangentMode=RCTM_Break, time=0.3, value=1.1}
          {arriveTangent=-4.5642, interpMode=RCIM_Cubic, leaveTangent=-3.2837, tangentMode=RCTM_Break, time=1, value=0}
         ]}}}
      CurveIndex=$Particles.Progress
     )))
   /SetVariables(Particles.MeshOrientationPak:Quat4f=$Particles.MeshOrientation, Particles.PositionBak:NiagaraPosition=$Particles.Position) {isSetParameters=true}
   &SystemLocation(
    MaskSpawnGroup=false
    Offset=&RotateVectorByQuaternion(
     Vector To Rotate=&AddVectorToPosition(
      Position=&Multiply_VectorByFloat(
       Vector=&AddVectorToPosition(
        Position=&Lerp_Vector(
         A=<0 0 -10>
         B=<0 -250 0>
         Alpha=&FloatFromCurve(
          FloatCurve=DataInterface() {propertyValues {Curve {keys [{interpMode=RCIM_Cubic, tangentMode=RCTM_Break, time=0, value=0}, {tangentMode=RCTM_User, time=1, value=1}]}}}
          CurveIndex=$Particles.Progress
         )
        )
        Vector=&Lerp_Vector(
         A=<0 0 0>
         B=&MaskVectorBySpawnGroup(Masked Vector=<65 0 0>, Passthrough Vector=<-65 1 1>)
         Alpha=&FloatFromCurve(
          FloatCurve=DataInterface() {propertyValues {Curve {keys [{interpMode=RCIM_Cubic, leaveTangent=3.2771, tangentMode=RCTM_Break, time=0, value=0}, {tangentMode=RCTM_User, time=1, value=1}]}}}
          CurveIndex=$Particles.Progress
         )
        )
       )
       Float=&FN_Fit01(Clamp:NiagaraBool=false, Value:NiagaraFloat=$Emitter.float, DestMin:NiagaraFloat=0.2, DestMax:NiagaraFloat=1)
      )
      Vector=$Particles.LSOffset
     )
     Quaternion=$System.Quat
    )
    Offset Coordinate Space=::World
   )
   /SetVariables(Particles.MeshOrientation=&MultiplyQuaternion(
     Quaternion A=&MultiplyQuaternion(
      Quaternion A=&MakeQuatFromAxisAngle(
       Axis=<0 1 0>
       AngleInDegrees=&Multiply_Float(
        A=&MaskFloatBySpawnGroup(Masked Float=35, Passthrough Float=-35)
        B=&FloatFromCurve(
         FloatCurve=DataInterface() {propertyValues {Curve {keys [<at 0 1>, {tangentMode=RCTM_User, time=1, value=1}]}}}
         CurveIndex=$Particles.Progress
        )
       )
      )
      Quaternion B=&FindQuatBetweenTwoVectors(
       From Vector=<0 0 1>
       To Vector=/Niagara/DynamicInputs/Math/SlerpVector(Vector A:Vector3f=<0 -0.7 0.7>, Vector B:Vector3f=<0 -1 0>, Alpha:NiagaraFloat=$Particles.Progress, Scale:NiagaraFloat=1)
      )
     )
     Quaternion B=&FindQuatBetweenTwoVectors(From Vector=<0 0 1>, To Vector=&RandomRangeVector(Minimum=<-0.1 -0.1 1>, Maximum=<0.1 0.1 1>))
    )) {isSetParameters=true}
   /SetVariables(Particles.MeshOrientation=&FindQuatBetweenTwoVectors(
     From Vector=<0 0 1>
     To Vector=&RotateVectorByQuaternion(
      Vector To Rotate=&RotateVectorByQuaternion(Vector To Rotate=<0 0 1>, Quaternion=$Particles.MeshOrientation)
      Quaternion=$System.Quat
     )
    )) {isSetParameters=true}
   /SetVariables(Particles.Progress:NiagaraFloat=&Frac_Float(Float=&Add_Float(
      A=$Particles.Progress
      B=&Multiply_Float(
       A=&VectorLength(
        Vector=&Multiply_VectorByFloat(Vector=&Multiply_VectorByFloat(Vector=$Engine.Owner.Velocity, Float=-1), Float=$Engine.DeltaTime)
        Safe Length Calculation=false
       )
       B=&Multiply_Float(A=0.0008, B=&Multiply_Float(A=$System.M, B=1))
      )
     ))) {isSetParameters=true}
   &DynamicMaterialParameters(
    Index 0 Param 1=&RandomRangeFloat(Minimum=0, Maximum=100)
    Module.Index0Param0Write=true
    Module.Index0Param1Write=true
    Module.Index0Param2Write=true
    Module.Index0Param3Write=true
   )
   /SetVariables(
    Particles.MeshOrientation=&LerpQuaternion(
     Quaternion A:Quat4f=$Particles.MeshOrientationPak
     Quaternion B:Quat4f=$Particles.MeshOrientation
     Lerp Factor:NiagaraFloat=$System.M
     Enable Spherical Linear Interpolation:NiagaraBool=false
    )
    Particles.Position=&LerpPosition(A=$Particles.PositionBak, B=$Particles.Position, Alpha=$System.M)
   ) {isSetParameters=true}
   &ScaleMeshSize(Scale Factor=&VectorFromFloat(Value=&ClampFloat(Float=&Subtract_Float(
       A=&Add_Float(A=&Subtract_Float(A=1, B=$Particles.Progress), B=1)
       B=&FloatFromCurve(FloatCurve=DataInterface() {propertyValues {Curve {keys [<at 1.45 0>,<at 1.5 2>]}}}, CurveIndex=$System.Timeline001.Time)
      )))) {module=ScaleMeshSize003}
   &ScaleMeshSize(Scale Factor=&VectorFromFloat(Value=$Particles.Age)) {enabled=false, module=ScaleMeshSize002}
   &KillParticles(Kill Particles=&SetBoolByFloatComparison(A=$Particles.Age, B=5))
   &ScaleColor(
    Scale Alpha=&ClampFloat(Float=&Subtract_Float(
      A=&Add_Float(A=&Subtract_Float(A=1, B=$Particles.Progress), B=1)
      B=&FloatFromCurve(
       FloatCurve=DataInterface() {propertyValues {Curve {keys [
           {interpMode=RCIM_Cubic, leaveTangent=2.4051, tangentMode=RCTM_Break, time=3, value=0}
           {arriveTangent=0.7311, tangentMode=RCTM_User, time=4, value=2}
          ]}}}
       CurveIndex=$Particles.Age
      )
     ))
    Scale RGB=&VectorFromFloat(Value=&Lerp_Float(
      A=&FloatFromCurve(
       FloatCurve=DataInterface() {propertyValues {Curve {keys [{interpMode=RCIM_Cubic, leaveTangent=-18.599, tangentMode=RCTM_Break, time=0, value=7}, <at 1 1>]}}}
       CurveIndex=$Particles.Progress
      )
      B=&FloatFromCurve(
       FloatCurve=DataInterface() {propertyValues {Curve {keys [{interpMode=RCIM_Cubic, leaveTangent=-1.8783, tangentMode=RCTM_Break, time=0, value=2}, <at 1 1>]}}}
       CurveIndex=$Particles.Progress
      )
      Alpha=$Emitter.float001
     ))
    ScaleA=true
    ScaleRGB=true
   )
   &FN_Sprite_Attributes(
    Color=&MultiplyLinearColor(
     Linear Color=&MultiplyLinearColor(
      Linear Color=<rgba 0.3481 0.7018 1 1>
      Scale Factor=&MakeFloat4FromVectorAndFloat(
       "Vector (RGB)"=&VectorFromFloat(Value=&Lerp_Float(
         A=&FloatFromCurve(
          FloatCurve=DataInterface() {propertyValues {Curve {keys [
              {interpMode=RCIM_Cubic, leaveTangent=-12.2299, tangentMode=RCTM_Break, time=0, value=4}
              {arriveTangent=-2.3164, interpMode=RCIM_Cubic, leaveTangent=-2.3164, tangentMode=RCTM_User, time=0.1, value=1.75}
              <at 1 0.3>
             ]}}}
          CurveIndex=$Particles.Progress
         )
         B=&FloatFromCurve(
          FloatCurve=DataInterface() {propertyValues {Curve {keys [{arriveTangent=-0.7593, interpMode=RCIM_Cubic, leaveTangent=-0.7593, tangentMode=RCTM_User, time=0, value=0.3}, <at 1 0>]}}}
          CurveIndex=$Particles.Progress
         )
         Alpha=$Emitter.float001
        ))
       "Float (Alpha)"=$Emitter.float
      )
     )
     Scale Factor=&MakeVector4(
      X=1
      Y=1
      Z=1
      W=&Multiply_Float(
       A=&ClampFloat(Float=&Subtract_Float(
         A=&Add_Float(A=&Subtract_Float(A=1, B=$Particles.Progress), B=1)
         B=&FloatFromCurve(
          FloatCurve=DataInterface() {propertyValues {Curve {keys [
              {interpMode=RCIM_Cubic, leaveTangent=2.4051, tangentMode=RCTM_Break, time=3, value=0}
              {arriveTangent=0.7311, tangentMode=RCTM_User, time=4, value=2}
             ]}}}
          CurveIndex=$Particles.Age
         )
        ))
       B=0.12
      )
     )
    )
    Create Color Attribute=true
    Create SpriteSize Attribute=true
    SpriteSize=&Multiply_Vector2D_ByFloat(
     Float=&FloatFromCurve(
      FloatCurve=DataInterface() {propertyValues {Curve {keys [
          {interpMode=RCIM_Cubic, leaveTangent=8.0598, tangentMode=RCTM_Break, time=0, value=0}
          {arriveTangent=-0.9971, interpMode=RCIM_Cubic, leaveTangent=-2.726, tangentMode=RCTM_Break, time=0.3, value=1.1}
          {arriveTangent=-4.5642, interpMode=RCIM_Cubic, leaveTangent=-3.2837, tangentMode=RCTM_Break, time=1, value=0}
         ]}}}
      CurveIndex=$Particles.Progress
      Scale Curve=&ClampFloat(Float=&Subtract_Float(
        A=&Add_Float(A=&Subtract_Float(A=1, B=$Particles.Progress), B=1)
        B=&FloatFromCurve(FloatCurve=DataInterface() {propertyValues {Curve {keys [<at 2.5 0>,<at 4 2>]}}}, CurveIndex=$Particles.Age)
       ))
     )
     Vector2D=<60 150>
    )
    Write to Module Namespace=true
   )
   /SetVariables(Particles.Velocity=&RotateVectorByQuaternion(Vector To Rotate=<0 0 1>, Quaternion=$Particles.MeshOrientation)) {isSetParameters=true}
  ]
  rendererClasses [&NiagaraSpriteRendererProperties, &NiagaraMeshRendererProperties]
  renderers [
   {rendererClass=&NiagaraSpriteRendererProperties}
   {rendererClass=&NiagaraMeshRendererProperties, rendererIndex=1}
   {rendererClass=&NiagaraSpriteRendererProperties, rendererIndex=2}
  ]
 }
 Smoke {
  emitterUpdateScript [
   &EmitterState()
   &SpawnBurst_Instantaneous(Spawn Count=2, Use Spawn Probability=false, UseLoopCountLimit=false)
   &SpawnBurst_Instantaneous(Spawn Count=2, Spawn Group=1, Use Spawn Probability=false, UseLoopCountLimit=false) {module=SpawnBurst_Instantaneous001}
   &SpawnPerUnit(
    Module.Use Max Movement Threshold=true
    Module.Use Movement Tolerance=false
    Spawn Spacing:NiagaraFloat=&MakeCustomFloatFromBool(Bool=&SetBoolByFloatComparison(A=$User.Moving, B=0.9), True Float=300, False Float=0)
    Use Spawn Probability=false
   )
   &SpawnPerUnit(
    Module.Use Max Movement Threshold=true
    Module.Use Movement Tolerance=false
    Spawn Group=1
    Spawn Spacing:NiagaraFloat=&MakeCustomFloatFromBool(Bool=&SetBoolByFloatComparison(A=$User.Moving, B=0.9), True Float=300, False Float=0)
    Use Spawn Probability=false
   ) {module=SpawnPerUnit001}
   &SpawnPerUnit(
    Module.Use Max Movement Threshold=true
    Module.Use Movement Tolerance=false
    Spawn Spacing:NiagaraFloat=&MakeCustomFloatFromBool(Bool=&SetBoolByFloatComparison(A=$User.Moving, B=0.9), True Float=300, False Float=0)
    Use Spawn Probability=false
   ) {module=SpawnPerUnit002}
   &SpawnPerUnit(
    Module.Use Max Movement Threshold=true
    Module.Use Movement Tolerance=false
    Spawn Group=1
    Spawn Spacing:NiagaraFloat=&MakeCustomFloatFromBool(Bool=&SetBoolByFloatComparison(A=$User.Moving, B=0.9), True Float=300, False Float=0)
    Use Spawn Probability=false
   ) {module=SpawnPerUnit003}
  ]
  particleSpawnScript [
   /Niagara/Modules/Spawn/Initialization/InitializeParticle(
    Color=&MakeLinearColorFromVectorAndFloat(
     "Vector (RGB)"=&MakeVectorFromLinearColorRGB(Color=&RandomRangeLinearColor(Minimum=<rgba 0.3375 0.42 0.5625 1>, Maximum=<rgba 0.4367 0.5347 0.6719 1>))
     "Float (Alpha)"=&UniformRangedFloat(Minimum=1, Maximum=1.1)
    )
    Lifetime=&UniformRangedFloat(Minimum=0.2, Maximum=0.3)
    Module.Write Color=true
    Module.Write Lifetime=true
    Module.Write Mass=true
    Module.Write Material Random=true
    Module.Write Position=true
    Module.Write Scale=false
    Module.Write SpriteRotation=true
    Module.Write SpriteSize=true
    Position=&AddVectorToPosition(
     Position=&SimulationPosition()
     Vector=&TransformVector(
      Vector=&MaskVectorBySpawnGroup(Masked Vector=<45 120 -20>, Passthrough Vector=<-45 120 -20>)
      Destination Space=::Simulation
     )
    )
    Sprite Rotation=&NormalizedAngleToDegrees(NormalizedAngle=&UniformRangedFloat(Minimum=0, Maximum=1))
    Sprite Size=&Vector2DFromFloat(Value=&Multiply_Float(A=&UniformRangedFloat(Minimum=87.0012, Maximum=123.7263), B=0.4624))
   )
   &CameraOffset(Camera Offset Amount=&UniformRangedFloat(Minimum=5, Maximum=111.2439))
   &AddVelocityFromPoint(
    Module.Use Falloff Distance=false
    Origin Offset=<0 119.9256 -50>
    Velocity Strength=&RandomRangeFloat(Minimum=200, Maximum=500)
   ) {enabled=false}
   &AddVelocity(
    Cone Axis=&MaskVectorBySpawnGroup(Masked Vector=<1 -2 1>, Passthrough Vector=<-1 -2 1>)
    Module.UseConeFalloff=false
    Module.UseVelDistribution=false
    Override Seed=false
    Velocity Mode=::NewEnumerator2
    Velocity Speed=&RandomRangeFloat(Minimum=1253.9437, Maximum=1757.735)
   )
  ]
  particleUpdateScript [
   &ParticleState()
   &SpriteRotationRate(Rotation Rate=&Multiply_Float(
     A=$Particles.InitSpriteRotRate
     B=&FloatFromCurve(FloatCurve=DataInterface() {propertyValues {Curve {keys [<at -0.0046 1.0113 RCIM_Cubic>,<at 1.0011 0.09 RCIM_Cubic>]}}})
    ))
   /Niagara/Modules/Update/SubUV/SubUVAnimation(Number Of Frames=64, SubUV Lookup Index=&Multiply_Float(A=$Particles.Age, B=0.66))
   &Drag(Drag=6, Use Linear Drag=true, Use Rotational Drag=false)
   &SolveForcesAndVelocity(Clamp Velocity=false, Limit Acceleration=false)
   &Color(
    Color=$Particles.Initial.Color
    Scale Alpha=&FloatFromCurve(FloatCurve=DataInterface() {propertyValues {Curve {keys [{arriveTangent=-0.8206, interpMode=RCIM_Cubic, leaveTangent=-0.8206, tangentMode=RCTM_User, time=0, value=0.5}, <at 1.0011 0.002 RCIM_Cubic>]}}})
    Scale Color=&VectorFromFloat(Value=1)
   )
   &ScaleSpriteSize(Scale Factor=&Vector2DFromFloat(Value=&FloatFromCurve(FloatCurve=DataInterface() {propertyValues {Curve {keys [
          {arriveTangent=6.716, interpMode=RCIM_Cubic, leaveTangent=6.4748, tangentMode=RCTM_User, time=-0.003, value=0}
          {interpMode=RCIM_Cubic, leaveTangent=6.3679, time=1, value=3}
         ]}}})))
   &DynamicMaterialParameters(
    Index 0 Param 1=&RandomRangeFloat(Minimum=5, Maximum=1)
    Index 0 Param 2=1
    Index 0 Param 3=1
    Index 0 Param 4=1
    Module.Param0WriteEnabled=true
    Module.Param1WriteEnabled=true
    Module.Param2WriteEnabled=true
    Module.Param3WriteEnabled=true
   )
  ]
  rendererClasses [&NiagaraSpriteRendererProperties]
  renderers [{rendererClass=&NiagaraSpriteRendererProperties}]
 }
 Sparks001 {
  simTarget=GPUComputeSim
  emitterUpdateScript [
   &EmitterState()
   &SpawnBurst_Instantaneous(Spawn Count=20, Use Spawn Probability=false, UseLoopCountLimit=false)
   &SpawnBurst_Instantaneous(Spawn Count=20, Spawn Group=1, Use Spawn Probability=false, UseLoopCountLimit=false) {module=SpawnBurst_Instantaneous001}
   &SpawnPerUnit(
    Module.Use Max Movement Threshold=true
    Module.Use Movement Tolerance=false
    Spawn Spacing:NiagaraFloat=&MakeCustomFloatFromBool(Bool=&SetBoolByFloatComparison(A=$User.Moving, B=0.9), True Float=5, False Float=0)
    Use Spawn Probability=false
   )
   &SpawnPerUnit(
    Module.Use Max Movement Threshold=true
    Module.Use Movement Tolerance=false
    Spawn Group=1
    Spawn Spacing:NiagaraFloat=&MakeCustomFloatFromBool(Bool=&SetBoolByFloatComparison(A=$User.Moving, B=0.9), True Float=5, False Float=0)
    Use Spawn Probability=false
   ) {module=SpawnPerUnit001}
  ]
  particleSpawnScript [
   /Niagara/Modules/Spawn/Initialization/V2/InitializeParticle(
    Color=<rgba 1 1.3333 2 1>
    Lifetime=&RandomRangeFloat(Minimum=0.1, Maximum=0.3)
    Mesh Scale Mode=::NewEnumerator4
    Position Offset=&TransformVector(
     Vector=&MaskVectorBySpawnGroup(Masked Vector=<45 120 0>, Passthrough Vector=<-45 120 0>)
     Destination Space=::Simulation
    )
    Sprite Size=&RandomRangeVector2D(Minimum=<5 8>, Maximum=<7 20>)
    Sprite Size Mode=::NewEnumerator4
    UsePositionOffset=true
    Write Material Random=true
   )
   &ShapeLocation(MaskSpawnGroup=false, Override Seed=false, Sphere Radius=0.5, UseOwnerScale=true)
   &AddVelocity(
    Module.Use Falloff Distance=false
    Origin Offset=<0 0.4679 -0.7128>
    Override Seed=false
    Velocity Mode=::NewEnumerator1
    Velocity Speed=&RandomRangeFloat(Minimum=1955.4866, Maximum=3610.7502)
   ) {enabled=false}
   &AddVelocity(
    Cone Angle=70.7827
    Cone Axis=&MaskVectorBySpawnGroup(Masked Vector=<0.5 -2 1>, Passthrough Vector=<-0.5 -2 1>)
    Module.UseConeFalloff=false
    Module.UseVelDistribution=false
    Override Seed=false
    Velocity Mode=::NewEnumerator2
    Velocity Speed=&RandomRangeFloat(Minimum=1774.5864, Maximum=3912.3142)
   ) {module=AddVelocity001}
   /SetVariables(Particles.Velocity=&Multiply_Vector(A=$Particles.Velocity, B=<1 1 0.7>)) {isSetParameters=true}
   &DynamicMaterialParameters(
    Index 0 Param 2=1
    Index 0 Param 3=1
    Index 0 Param 4=0.03
    Index 1 Param 2=0.15
    Index 1 Param 3=0.5
    Module.Index0Float0Write=false
    Module.Index0Float1Write=true
    Module.Index0Float2Write=true
    Module.Index0Float3Write=true
    Module.Index1Float0Write=true
    Module.Index1Float1Write=true
    Module.Index1Float2Write=true
    Module.Index1Float3Write=false
    Write Parameter Index 1=::NewEnumerator1
   )
  ]
  particleUpdateScript [
   &Drag(
    Drag=&RandomRangeFloat(Minimum=10.5722, Maximum=19.8375)
    Use Linear Drag=true
    Use Rotational Drag=true
   )
   &SolveForcesAndVelocity(Clamp Velocity=false, Limit Acceleration=false)
   &ParticleState()
   &ScaleSpriteSizeBySpeed(Max Scale Factor=<1 2>, Min Scale Factor=<1 0.25>, Velocity Threshold=281.7375)
   &ScaleSpriteSize(
    Scale Factor=&MakeVector2D(
     X=1
     Y=&FloatFromCurve(FloatCurve=DataInterface() {propertyValues {Curve {keys [<at 0.8 1>,<at 1 0>]}}})
    )
    Scale Sprite Size Mode=::NewEnumerator2
   )
   &FN_Sprite_Attributes(
    Color=&MultiplyLinearColor(Linear Color=$Particles.Color, Scale Factor=<1 1 1 0.05>)
    Create Color Attribute=true
    Create SpriteSize Attribute=true
    SpriteSize=&Multiply_Vector2D(A=$Particles.SpriteSize, B=<15 2>)
    Write to Module Namespace=true
   )
  ]
  rendererClasses [&NiagaraSpriteRendererProperties]
  renderers [{rendererClass=&NiagaraSpriteRendererProperties}]
 }
 TipDecalNoSpawnUnit {
  emitterUpdateScript [
   &EmitterState(Life Cycle Mode=::NewEnumerator1, Loop Behavior=::NewEnumerator1, Loop Duration=1, UseLoopDelay=false)
   &SpawnBurst_Instantaneous(Use Spawn Probability=false, UseLoopCountLimit=false)
   /SetVariables(Emitter.GradTime=0.5, Emitter.MinWidth=-0.4) {isSetParameters=true, module=SetVariables_#8}
  ]
  particleSpawnScript [
   &InitializeParticle(
    Lifetime=5
    Mesh Scale=<3 3 1>
    Mesh Scale Mode=::NewEnumerator4
    Ribbon Width=187.511
    Ribbon Width Mode=::NewEnumerator1
    UsePositionOffset=false
    Write Material Random=true
   )
   &SkeletalMeshLocation(
    Bone Sampling Mode=::NewEnumerator4
    MaskSpawnGroup=false
    Orientation Sampling=::NewEnumerator1
    Sampled Position Offset=<0 60 0>
    Skeletal Mesh=$System.TargetSKMesh
   ) {enabled=false}
   &ShapeLocation(
    MaskSpawnGroup=false
    Non Uniform Scale=<1 1 0>
    Offset=<0 60 0>
    Offset Coordinate Space=::Local
    Offset Mode=::NewEnumerator0
    Override Seed=false
    Sphere Radius=0.0001
    UseOwnerScale=true
   )
   /SetVariables(Particles.LSOffset:Vector3f=&TransformPosition(Position=$Particles.Initial.ShapeLocation.ShapePosition, Source Space=::World, Destination Space=::Local)) {isSetParameters=true}
   &InitialMeshOrientation(Mesh Orientation Mode=::NewEnumerator1, Use Rotation Vector=false)
   &self:ScratchModule() {enabled=false, module=ScratchModule}
   &DynamicMaterialParameters(
    Index 2 Param 2=$Particles.F
    Module.Index2Param0Write=true
    Module.Index2Param1Write=true
    Module.Index2Param2Write=true
    Module.Index2Param3Write=true
    Write Parameter Index 0=false
    Write Parameter Index 2=true
   ) {module=DynamicMaterialParameters001}
  ]
  particleUpdateScript [
   &ParticleState()
   /SetVariables(Particles.MeshOrientationPak:Quat4f=$Particles.MeshOrientation, Particles.PositionBak:NiagaraPosition=$Particles.Position) {isSetParameters=true}
   &self:QuatToAxisAngle(Quat:Quat4f=$Particles.MeshOrientation) {module=QuatToAxisAngle}
   &self:SetTipAndTailOpacity() {module=SetTipAndTailOpacity}
   &DynamicMaterialParameters(
    Index 0 Param 1=&MakeFloatFromVector(Vector=$Particles.Axis, Channel=::NewEnumerator0)
    Index 0 Param 2=&MakeFloatFromVector(Vector=$Particles.Axis, Channel=::NewEnumerator1)
    Index 0 Param 3=&MakeFloatFromVector(Vector=$Particles.Axis, Channel=::NewEnumerator2)
    Index 0 Param 4=$Particles.Angle
    Index 1 Param 1=&Multiply_Float(
     A=200
     B=&FloatFromCurve(
      FloatCurve=DataInterface() {propertyValues {Curve {keys [
          {interpMode=RCIM_Cubic, leaveTangent=2.9289, tangentMode=RCTM_Break, time=0, value=0}
          {arriveTangent=0.4353, tangentMode=RCTM_User, time=0.5, value=1}
         ]}}}
      CurveIndex=$Emitter.Age
     )
    )
    Index 1 Param 2=&Multiply_Float(
     A=250
     B=&FloatFromCurve(
      FloatCurve=DataInterface() {propertyValues {Curve {keys [
          {interpMode=RCIM_Cubic, leaveTangent=2.9289, tangentMode=RCTM_Break, time=0, value=0}
          {arriveTangent=0.4353, tangentMode=RCTM_User, time=0.5, value=1}
         ]}}}
      CurveIndex=$Emitter.Age
     )
    )
    Index 1 Param 3=100
    Index 1 Param 4=0.1
    Index 2 Param 2=$Particles.F
    Module.Index0Param0Write=true
    Module.Index0Param1Write=true
    Module.Index0Param2Write=true
    Module.Index0Param3Write=true
    Module.Index1Param0Write=true
    Module.Index1Param1Write=true
    Module.Index1Param2Write=true
    Module.Index1Param3Write=true
    Module.Index2Param0Write=true
    Module.Index2Param1Write=true
    Module.Index2Param2Write=true
    Module.Index2Param3Write=true
    Write Parameter Index 1=true
    Write Parameter Index 2=true
   )
   &ScaleColor(
    Scale Alpha=&FloatFromCurve(
     FloatCurve=DataInterface() {propertyValues {Curve {keys [
         {interpMode=RCIM_Cubic, leaveTangent=4.5617, tangentMode=RCTM_Break, time=0, value=0}
         {arriveTangent=0.4353, tangentMode=RCTM_User, time=0.5, value=1}
        ]}}}
     CurveIndex=&Divide_Float(A=$Emitter.Age, B=$Emitter.GradTime)
    )
    ScaleA=true
    ScaleRGB=true
   )
   &SystemLocation(MaskSpawnGroup=false, Offset=$Particles.LSOffset)
   /SetVariables(
    Particles.MeshOrientation=$Particles.MeshOrientation
    Particles.Position=&LerpPosition(A=$Particles.PositionBak, B=$Particles.Position, Alpha=$System.M)
   ) {isSetParameters=true}
   &ScaleColor(
    Scale Alpha=&FloatFromCurve(FloatCurve=DataInterface() {propertyValues {Curve {keys [<at 1.5 1>,<at 1.9 0>]}}}, CurveIndex=$System.Timeline001.Time)
    ScaleA=true
    ScaleRGB=true
   ) {module=ScaleColor001}
  ]
  rendererClasses [&NiagaraRibbonRendererProperties, &NiagaraMeshRendererProperties]
  renderers [
   {rendererClass=&NiagaraRibbonRendererProperties}
   {rendererClass=&NiagaraMeshRendererProperties, rendererIndex=1}
   {rendererClass=&NiagaraMeshRendererProperties, rendererIndex=2}
  ]
 }
}
summary {
 systemName=NS_MotoWrestle_Slide
 emitters [
  {emitterName=IceDecal, rendererClasses [&NiagaraRibbonRendererProperties, &NiagaraMeshRendererProperties]}
  {emitterName=BehindSmoke, rendererClasses [&NiagaraSpriteRendererProperties]}
  {emitterName=BehindIcePiece, rendererClasses [&NiagaraMeshRendererProperties]}
  {emitterName=RandomSpike, rendererClasses [&NiagaraSpriteRendererProperties, &NiagaraMeshRendererProperties]}
  {
   emitterName=TipDecalNoSpawnUnit
   rendererClasses [&NiagaraRibbonRendererProperties, &NiagaraMeshRendererProperties]
  }
  {emitterName=FrostDecal, rendererClasses [&NiagaraRibbonRendererProperties, &NiagaraMeshRendererProperties]}
  {emitterName=Smoke, rendererClasses [&NiagaraSpriteRendererProperties]}
  {emitterName=Blast, rendererClasses [&NiagaraSpriteRendererProperties]}
  {emitterName=GlowFlick, rendererClasses [&NiagaraSpriteRendererProperties]}
  {emitterName=RollingMesh, rendererClasses [&NiagaraSpriteRendererProperties, &NiagaraMeshRendererProperties]}
  {emitterName=EnergyFlow, rendererClasses [&NiagaraSpriteRendererProperties]}
  {emitterName=Sparks001, simTarget=GPUComputeSim, rendererClasses [&NiagaraSpriteRendererProperties]}
 ]
 userVariables [
  User.ColorVariant {type=NiagaraInt32}
  User.float {type=NiagaraFloat, defaultValue=0, description=""}
  User.Moving {type=NiagaraFloat, defaultValue=1, description=""}
 ]
}
