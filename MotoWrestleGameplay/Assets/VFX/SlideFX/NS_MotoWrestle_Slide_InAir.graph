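# Niagara effect for the MotoWrestle in-air slide, with TipIceMesh as the only emitter that has renderers or visible script content.
# TipIceMesh continuously spawns short-lived, randomly sized meshes around an offset cylinder surface, aligned to a point-based velocity that never moves them, as the force solver is disabled.
# Curves grow, shrink and fade each mesh, while a module writes separate light-blue, low-alpha sprite attributes for an accompanying sprite renderer.
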NiagaraSystem /MotoWrestleGameplay/Assets/VFX/SlideFX/NS_MotoWrestle_Slide_InAir
dependencies {usedDataInterfaces [&NiagaraDataInterfaceCurve], usedDynamicInputs [], usedModules [&SystemState]}
emitters {
 FollowIceMesh {rendererClasses [], renderers []}
 FollowIceMesh001 {rendererClasses [], renderers []}
 Smoke002 {rendererClasses [], renderers []}
 Sparks {rendererClasses [], renderers []}
 TipIceMesh {
  emitterUpdateScript [
   &EmitterState(Life Cycle Mode=::NewEnumerator1, Loop Behavior=::NewEnumerator1, Loop Duration=50, UseLoopDelay=false)
   &SpawnRate(SpawnRate=100, Use Spawn Probability=false)
   &SpawnBurst_Instantaneous(Spawn Count=20, Use Spawn Probability=false, UseLoopCountLimit=false) {enabled=false}
  ]
  particleSpawnScript [
   /SetVariables(
    Particles.Vector:Vector3f=<0 10 0>
    Particles.Vector001:Vector3f=&Multiply_VectorByFloat(Vector=<10 15 -15.8041>, Float=0.5)
   ) {isSetParameters=true}
   &InitializeParticle(
    Color Mode=::NewEnumerator0
    Lifetime=&RandomRangeFloat(Minimum=0.3, Maximum=0.4)
    Mesh Scale Mode=::NewEnumerator3
    Mesh Uniform Scale=&RandomRangeFloat(Minimum=0.65, Maximum=0.75)
    Position Offset=$Particles.Vector
    UsePositionOffset=true
    Write Material Random=true
   )
   &ShapeLocation(
    Cylinder Height=20
    Cylinder Radius=7
    MaskSpawnGroup=false
    Override Seed=false
    Shape Primitive=::NewEnumerator1
    Surface Only=true
    Surface Only Band Thickness=1
    Use Endcaps In Surface Only Mode=false
    UseOwnerScale=true
   )
   &AddVelocity(
    Cone Angle=180
    Cone Axis=<0 -1 0>
    Module.UseConeFalloff=false
    Module.UseVelDistribution=false
    Override Seed=false
    Velocity Mode=::NewEnumerator2
    Velocity Speed=&RandomRangeFloat(Minimum=10, Maximum=40)
   ) {enabled=false, module=AddVelocity001}
   &AddVelocity(
    Module.Use Falloff Distance=false
    Origin Offset=&AddVectorToPosition(Position=Linked() {linkedVariable {name=Particles.Vector, type=Vector3f}}, Vector=$Particles.Vector001)
    Origin Offset Coordinate Space=::Simulation
    Override Seed=false
    Velocity Mode=::NewEnumerator1
    Velocity Speed=10
   )
   &InitialMeshOrientation(Mesh Axis To Orient=<0 0 1>, Mesh Orientation Mode=::NewEnumerator2, Orientation Vector=$Particles.Velocity, Use Rotation Vector=false)
   &CameraOffset(Camera Offset Amount=10)
  ]
  particleUpdateScript [
   &SolveForcesAndVelocity(Clamp Velocity=false, Limit Acceleration=false) {enabled=false}
   &ParticleState()
   &ScaleMeshSize(Scale Factor=&VectorFromFloat(Value=&FloatFromCurve(FloatCurve=DataInterface() {propertyValues {Curve {keys [
          {arriveTangent=3.768, interpMode=RCIM_Cubic, leaveTangent=3.768, tangentMode=RCTM_User, time=0, value=0}
          {arriveTangent=-0.3293, interpMode=RCIM_Cubic, leaveTangent=0.1619, tangentMode=RCTM_Break, time=0.3, value=0.9847}
          {arriveTangent=-0.0405, interpMode=RCIM_Cubic, leaveTangent=-0.0405, tangentMode=RCTM_User, time=1, value=0}
         ]}}}))) {module=ScaleMeshSize001}
   &ScaleMeshSize(Scale Factor=<0.5 0.5 1>)
   &ScaleColor(
    Scale Alpha=&FloatFromCurve(FloatCurve=DataInterface() {propertyValues {Curve {keys [
         {arriveTangent=-0.3757, interpMode=RCIM_Cubic, leaveTangent=-0.3757, tangentMode=RCTM_User, time=0.4, value=0.9847}
         {arriveTangent=-0.0405, interpMode=RCIM_Cubic, leaveTangent=-0.0405, tangentMode=RCTM_User, time=1, value=0}
        ]}}})
    Scale RGB=&VectorFromFloat(Value=&FloatFromCurve(FloatCurve=DataInterface() {propertyValues {Curve {keys [{interpMode=RCIM_Cubic, leaveTangent=-0.2686, tangentMode=RCTM_Break, time=0, value=3}, <at 0.7 1>]}}}))
    ScaleA=true
    ScaleRGB=true
   )
   /SetVariables() {isSetParameters=true}
   &FN_Sprite_Attributes(
    Color=&MultiplyLinearColor(
     Linear Color=<rgba 0.3481 0.7018 1 1>
     Scale Factor=&MakeVector4(
      X=1
      Y=1
      Z=1
      W=&Multiply_Float(
       A=0.07
       B=&FloatFromCurve(FloatCurve=DataInterface() {propertyValues {Curve {keys [{interpMode=RCIM_Cubic, tangentMode=RCTM_Break, time=0.5, value=1}, <at 1 0>]}}})
      )
     )
    )
    Create Color Attribute=true
    Create SpriteSize Attribute=true
    SpriteSize=&Multiply_Vector2D(
     A=&Multiply_Vector2D_ByFloat(
      Float=&FloatFromCurve(FloatCurve=DataInterface() {propertyValues {Curve {keys [{interpMode=RCIM_Cubic, tangentMode=RCTM_Break, time=0.5, value=1}, <at 1 0>]}}})
      Vector2D=<1 1>
     )
     B=<43.2386 202.4155>
    )
    Write to Module Namespace=true
   )
  ]
  rendererClasses [&NiagaraSpriteRendererProperties, &NiagaraMeshRendererProperties]
  renderers [
   {rendererClass=&NiagaraSpriteRendererProperties}
   {rendererClass=&NiagaraMeshRendererProperties, rendererIndex=1}
  ]
 }
 Upward_Energy001 {rendererClasses [], renderers []}
}
summary {
 systemName=NS_MotoWrestle_Slide_InAir
 emitters [
  {emitterName=TipIceMesh, rendererClasses [&NiagaraSpriteRendererProperties, &NiagaraMeshRendererProperties]}
  {emitterName=Upward_Energy001, rendererClasses []}
  {emitterName=FollowIceMesh, rendererClasses []}
  {emitterName=FollowIceMesh001, rendererClasses []}
  {emitterName=Smoke002, rendererClasses []}
  {emitterName=Sparks, rendererClasses []}
 ]
 userVariables [User.Moving {type=NiagaraFloat, defaultValue=1, description=""}]
}
